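# Physics asset for the Teriyaki Warrior backpack cosmetic, defining bodies and joints for dangling slimey pieces and their headband tails.
# Four slimey capsules (A, B, D, E) each link through a small base box to the pack body.
# Base-to-pack joints leave swing and twist free, while slimey-to-base joints carry angular limits.
# Headband boxes form limited one- or two-link chains off slimeys A, D and E; B has none.
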PhysicsAsset /BRCosmetics/Accessories/FORT_Backpacks/Backpack_M_MED_Teriyaki_Warrior/Meshes/M_MED_Teriyaki_Warrior_Pack_Physics
bodies {
 headbandAA_01 [{shapeType=Box, extentX=3.2593, extentY=0.5002, extentZ=1.01, center=<1.0995 0.0352 0.049>, rotation=<rot 0 0 90>}]
 headbandAB_01 [{shapeType=Box, extentX=3.2593, extentY=0.5002, extentZ=1.01, center=<1.1167 0 0>, rotation=<rot 0 0 90>}]
 headbandDA_01 [{shapeType=Box, extentX=2.6601, extentY=0.5019, extentZ=1.4854, center=<1 0 0>, rotation=<rot 0 0 90>}]
 headbandDA_02 [{shapeType=Box, extentX=2.6601, extentY=0.5019, extentZ=1.4854, center=<1 0 0>, rotation=<rot 0 0 90>}]
 headbandDB_01 [{shapeType=Box, extentX=2.6601, extentY=0.5019, extentZ=1.4854, center=<1 0 0>, rotation=<rot 0 0 90>}]
 headbandDB_02 [{shapeType=Box, extentX=2.6601, extentY=0.5019, extentZ=1.4854, center=<1 0 0>, rotation=<rot 0 0 90>}]
 headbandEA_01 [{shapeType=Box, extentX=5.0485, extentY=1.1759, extentZ=3.9246, center=<1 0 0>, rotation=<rot 0 0 90>}]
 headbandEA_02 [{shapeType=Box, extentX=3.7674, extentY=0.8006, extentZ=3.4002, center=<1 0 0>, rotation=<rot 0 0 90>}]
 headbandEB_01 [{shapeType=Box, extentX=3.896, extentY=0.5, extentZ=1.7188, center=<1.7795 0.0001 0>, rotation=<rot 0 0 90>}]
 headbandEB_02 [{shapeType=Box, extentX=4.1344, extentY=0.5, extentZ=1.7188, center=<1.8947 -0.1299 0>, rotation=<rot 0 -8.2599 89.9999>}]
 pack [{shapeType=Capsule, length=6.9682, radius=6.2455, center=<-11.158 -5.4983 0.41>}, {shapeType=Box, extentX=2, extentY=2, extentZ=2}]
 slimeyA_01 [{shapeType=Capsule, length=2.22, radius=2.301, center=<1.5943 0 0>, rotation=<rot 90 0 0>}]
 slimeyA_01_Base [{shapeType=Box, extentX=1.01, extentY=1.01, extentZ=1.01, rotation=<rot 0 -90 0>}]
 slimeyB_01 [{shapeType=Capsule, length=4.9627, radius=2.134, center=<2.5894 0.0272 -0.0818>, rotation=<rot 90 -0.0009 -0.0009>}]
 slimeyB_01_Base [{shapeType=Box, extentX=1.01, extentY=1.01, extentZ=1.01, rotation=<rot 0 -90 0>}]
 slimeyD_01 [{shapeType=Capsule, length=2.6593, radius=2.134, center=<2.5363 0.0272 0.1709>, rotation=<rot 83.7006 -180 -180>}]
 slimeyD_01_Base [{shapeType=Box, extentX=1.01, extentY=1.01, extentZ=1.01, rotation=<rot 0 -90 0>}]
 slimeyE_01 [{shapeType=Capsule, length=6.8804, radius=5.2181, center=<7.1756 1.3702 0>, rotation=<rot 89.29 0 0>}]
 slimeyE_01_Base [{shapeType=Box, extentX=1.01, extentY=1.01, extentZ=1.01, rotation=<rot 0 -90 0>}]
}
constraints [
 {bone1Name=slimeyE_01, bone2Name=slimeyE_01_Base, swing1LimitDegrees=20, swing2LimitDegrees=16, twistLimitDegrees=16}
 {bone1Name=slimeyD_01, bone2Name=slimeyD_01_Base, swing1LimitDegrees=30, swing2LimitDegrees=12, twistLimitDegrees=12}
 {bone1Name=slimeyA_01, bone2Name=slimeyA_01_Base, swing1LimitDegrees=45, swing2LimitDegrees=16, twistLimitDegrees=16}
 {bone1Name=headbandEA_01, bone2Name=slimeyE_01, swing1LimitDegrees=20, swing2LimitDegrees=20, twistLimitDegrees=12}
 {bone1Name=headbandEA_02, bone2Name=headbandEA_01, swing1LimitDegrees=20, swing2LimitDegrees=20, twistLimitDegrees=12}
 {bone1Name=headbandEB_01, bone2Name=slimeyE_01, swing1LimitDegrees=20, swing2LimitDegrees=35, twistLimitDegrees=12}
 {bone1Name=headbandEB_02, bone2Name=headbandEB_01, swing1LimitDegrees=20, swing2LimitDegrees=20, twistLimitDegrees=12}
 {bone1Name=headbandDA_01, bone2Name=slimeyD_01, swing1LimitDegrees=20, swing2LimitDegrees=20, twistLimitDegrees=12}
 {bone1Name=headbandDA_02, bone2Name=headbandDA_01, swing1LimitDegrees=20, swing2LimitDegrees=20, twistLimitDegrees=12}
 {bone1Name=headbandDB_01, bone2Name=slimeyD_01, swing1LimitDegrees=20, swing2LimitDegrees=20, twistLimitDegrees=12}
 {bone1Name=headbandDB_02, bone2Name=headbandDB_01, swing1LimitDegrees=20, swing2LimitDegrees=20, twistLimitDegrees=12}
 {bone1Name=headbandAB_01, bone2Name=slimeyA_01, swing1LimitDegrees=20, swing2LimitDegrees=20, twistLimitDegrees=12}
 {bone1Name=headbandAA_01, bone2Name=slimeyA_01, swing1LimitDegrees=20, swing2LimitDegrees=20, twistLimitDegrees=12}
 {bone1Name=slimeyB_01, bone2Name=slimeyB_01_Base, swing1LimitDegrees=40, swing2LimitDegrees=30, twistLimitDegrees=16}
 {
  bone1Name=slimeyB_01_Base
  bone2Name=pack
  swing1LimitDegrees=45
  swing1Motion=Free
  swing2LimitDegrees=45
  swing2Motion=Free
  twistLimitDegrees=45
  twistMotion=Free
 }
 {
  bone1Name=slimeyD_01_Base
  bone2Name=pack
  swing1LimitDegrees=45
  swing1Motion=Free
  swing2LimitDegrees=45
  swing2Motion=Free
  twistLimitDegrees=45
  twistMotion=Free
 }
 {
  bone1Name=slimeyE_01_Base
  bone2Name=pack
  swing1LimitDegrees=45
  swing1Motion=Free
  swing2LimitDegrees=45
  swing2Motion=Free
  twistLimitDegrees=45
  twistMotion=Free
 }
 {
  bone1Name=slimeyA_01_Base
  bone2Name=pack
  swing1LimitDegrees=45
  swing1Motion=Free
  swing2LimitDegrees=45
  swing2Motion=Free
  twistLimitDegrees=45
  twistMotion=Free
 }
]
