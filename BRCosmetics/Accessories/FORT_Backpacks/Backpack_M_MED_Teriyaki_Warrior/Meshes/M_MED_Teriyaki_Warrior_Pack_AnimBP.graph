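# Animation Blueprint for the Teriyaki Warrior backpack, procedurally animating its slime bones and playing a JoyJoy celebration on kills.
# RigidBody physics feeds additive piston jitter, constraints tying middle bones to their chain ends, and variable-driven stretch scaling.
# An interpolated FallingAlpha thins middle bones and extends tips.
# A rising AthenaKills stat blends to the JoyJoy sequence for a retriggerable few seconds, then resets dynamics.
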AnimBlueprint /BRCosmetics/Accessories/FORT_Backpacks/Backpack_M_MED_Teriyaki_Warrior/Meshes/M_MED_Teriyaki_Warrior_Pack_AnimBP
target_skeleton=&/M_MED_Teriyaki_Warrior_Pack_Skeleton
graphs [
 AnimGraph {nodes [
   n1 = RigidBody(ComponentPose=Local To Component(LocalPose=Input Pose())) {
    bClampLinearTranslationLimitToRefPose=true
    bOverrideWorldGravity=true
    overridePhysicsAsset=&/M_MED_Teriyaki_Warrior_Pack_Physics
    simulationSpace=BaseBoneSpace
    alphaBoolBlend {blendOption=Linear}
    baseBoneRef {boneName=root}
    componentAppliedLinearAccClamp=<1800 1800 1800>
    componentLinearAccScale=<0.1 0.1 0.5>
    componentLinearVelScale=<0.23 23 0.23>
    overrideWorldGravity {z=-350}
   }
   n2 = "Transform (Modify) Bone - Bone: slimeyE_03"(ComponentPose=@n1, Alpha=1, Translation=Get PistonJitterA() -> (PistonJitterA:Vector), Rotation="0,0,0", Scale="1,1,1") {translationMode=BMM_Additive, translationSpace=BCS_BoneSpace, alphaBoolBlend {blendOption=Linear}, boneToModify {boneName=slimeyE_03}}
   n3 = "Transform (Modify) Bone - Bone: slimeyD_03"(ComponentPose=@n2, Alpha=1, Translation=Get PistonJitterB() -> (PistonJitterB:Vector), Rotation="0,0,0", Scale="1,1,1") {translationMode=BMM_Additive, translationSpace=BCS_BoneSpace, alphaBoolBlend {blendOption=Linear}, boneToModify {boneName=slimeyD_03}}
   n4 = "Transform (Modify) Bone - Bone: slimeyA_03"(ComponentPose=@n3, Alpha=1, Translation=Get PistonJitterC() -> (PistonJitterC:Vector), Rotation="0,0,0", Scale="1,1,1") {translationMode=BMM_Additive, translationSpace=BCS_BoneSpace, alphaBoolBlend {blendOption=Linear}, boneToModify {boneName=slimeyA_03}}
   n5 = "Transform (Modify) Bone - Bone: slimeyB_03"(ComponentPose=@n4, Alpha=1, Translation=Get PistonJitterD() -> (PistonJitterD:Vector), Rotation="0,0,0", Scale="1,1,1") {translationMode=BMM_Additive, translationSpace=BCS_BoneSpace, alphaBoolBlend {blendOption=Linear}, boneToModify {boneName=slimeyB_03}}
   n6 = "Constraint - slimeyE_02 (PP)"!(ComponentPose=@n5, Alpha=1, ConstraintWeights_0=1, ConstraintWeights_1:float=1) -> (Pose) {
    alphaBoolBlend {blendOption=Linear}
    boneToModify {boneName=slimeyE_02}
    constraintSetup [{targetBone {boneName=slimeyE_01}}, {targetBone {boneName=slimeyE_03}}]
    constraintWeights [1,1]
   }
   n7 = "Constraint - slimeyD_02 (PP)"!(ComponentPose=@n6, Alpha=1, ConstraintWeights_0=1, ConstraintWeights_1:float=1) -> (Pose) {
    alphaBoolBlend {blendOption=Linear}
    boneToModify {boneName=slimeyD_02}
    constraintSetup [{targetBone {boneName=slimeyD_01}}, {targetBone {boneName=slimeyD_03}}]
    constraintWeights [1,1]
   }
   n8 = "Constraint - slimeyA_02 (PP)"!(ComponentPose=@n7, Alpha=1, ConstraintWeights_0=1, ConstraintWeights_1:float=1) -> (Pose) {
    alphaBoolBlend {blendOption=Linear}
    boneToModify {boneName=slimeyA_02}
    constraintSetup [{targetBone {boneName=slimeyA_01}}, {targetBone {boneName=slimeyA_03}}]
    constraintWeights [1,1]
   }
   n9 = "Constraint - slimeyB_02 (PP)"!(ComponentPose=@n8, Alpha=1, ConstraintWeights_0=1, ConstraintWeights_1:float=1) -> (Pose) {
    alphaBoolBlend {blendOption=Linear}
    boneToModify {boneName=slimeyB_02}
    constraintSetup [{targetBone {boneName=slimeyB_01}}, {targetBone {boneName=slimeyB_03}}]
    constraintWeights [1,1]
   }
   n10 = "Transform (Modify) Bone - Bone: slimeyE_02"(ComponentPose=@n9, Alpha=1, Translation="0,0,0", Rotation="0,0,0", Scale=Get StretchA() -> (StretchA:Vector)) {
    scaleMode=BMM_Replace
    scaleSpace=BCS_BoneSpace
    translationSpace=BCS_BoneSpace
    alphaBoolBlend {blendOption=Linear}
    boneToModify {boneName=slimeyE_02}
   }
   n11 = "Transform (Modify) Bone - Bone: slimeyD_02"(ComponentPose=@n10, Alpha=1, Translation="0,0,0", Rotation="0,0,0", Scale=Get StretchB() -> (StretchB:Vector)) {
    scaleMode=BMM_Replace
    scaleSpace=BCS_BoneSpace
    translationSpace=BCS_BoneSpace
    alphaBoolBlend {blendOption=Linear}
    boneToModify {boneName=slimeyD_02}
   }
   n12 = "Transform (Modify) Bone - Bone: slimeyA_02"(ComponentPose=@n11, Alpha=1, Translation="0,0,0", Rotation="0,0,0", Scale=Get StretchC() -> (StretchC:Vector)) {
    scaleMode=BMM_Replace
    scaleSpace=BCS_BoneSpace
    translationSpace=BCS_BoneSpace
    alphaBoolBlend {blendOption=Linear}
    boneToModify {boneName=slimeyA_02}
   }
   n13 = "Transform (Modify) Bone - Bone: slimeyB_02"(ComponentPose=@n12, Alpha=1, Translation="0,0,0", Rotation="0,0,0", Scale=Get StretchD() -> (StretchD:Vector)) {
    scaleMode=BMM_Replace
    scaleSpace=BCS_BoneSpace
    translationSpace=BCS_BoneSpace
    alphaBoolBlend {blendOption=Linear}
    boneToModify {boneName=slimeyB_02}
   }
   n14 = Get FallingAlpha() -> (FallingAlpha:float)
   n15 = "Transform (Modify) Bone - Bone: slimeyE_02"(ComponentPose=@n13, Alpha=@n14, Translation="0,0,0", Rotation="0,0,0", Scale="0.65,0.65,0.65") {
    scaleMode=BMM_Replace
    scaleSpace=BCS_BoneSpace
    translationSpace=BCS_BoneSpace
    alphaBoolBlend {blendOption=Linear}
    alphaScaleBiasClamp {bInterpResult=true, interpSpeedDecreasing=7, interpSpeedIncreasing=7}
    boneToModify {boneName=slimeyE_02}
    scale=<0.65 0.65 0.65>
   }
   n16 = "Transform (Modify) Bone - Bone: slimeyD_02"(ComponentPose=@n15, Alpha=@n14, Translation="0,0,0", Rotation="0,0,0", Scale="0.65,0.65,0.65") {
    scaleMode=BMM_Replace
    scaleSpace=BCS_BoneSpace
    translationSpace=BCS_BoneSpace
    alphaBoolBlend {blendOption=Linear}
    alphaScaleBiasClamp {bInterpResult=true, interpSpeedDecreasing=7, interpSpeedIncreasing=7}
    boneToModify {boneName=slimeyD_02}
    scale=<0.65 0.65 0.65>
   }
   n17 = "Transform (Modify) Bone - Bone: slimeyA_02"(ComponentPose=@n16, Alpha=@n14, Translation="0,0,0", Rotation="0,0,0", Scale="0.65,0.65,0.65") {
    scaleMode=BMM_Replace
    scaleSpace=BCS_BoneSpace
    translationSpace=BCS_BoneSpace
    alphaBoolBlend {blendOption=Linear}
    alphaScaleBiasClamp {bInterpResult=true, interpSpeedDecreasing=7, interpSpeedIncreasing=7}
    boneToModify {boneName=slimeyA_02}
    scale=<0.65 0.65 0.65>
   }
   n18 = "Transform (Modify) Bone - Bone: slimeyB_02"(ComponentPose=@n17, Alpha=@n14, Translation="0,0,0", Rotation="0,0,0", Scale="0.65,0.65,0.65") {
    scaleMode=BMM_Replace
    scaleSpace=BCS_BoneSpace
    translationSpace=BCS_BoneSpace
    alphaBoolBlend {blendOption=Linear}
    alphaScaleBiasClamp {bInterpResult=true, interpSpeedDecreasing=7, interpSpeedIncreasing=7}
    boneToModify {boneName=slimeyB_02}
    scale=<0.65 0.65 0.65>
   }
   n19 = "Transform (Modify) Bone - Bone: slimeyA_03"(ComponentPose=@n18, Alpha=@n14, Translation="4,0,0", Rotation="0,0,0", Scale="1,1,1") {
    rotationSpace=BCS_BoneSpace
    translationMode=BMM_Additive
    translationSpace=BCS_BoneSpace
    alphaBoolBlend {blendOption=Linear}
    alphaScaleBiasClamp {bInterpResult=true, interpSpeedDecreasing=7, interpSpeedIncreasing=4}
    boneToModify {boneName=slimeyA_03}
    translation {x=4}
   }
   n20 = "Transform (Modify) Bone - Bone: slimeyB_03"(ComponentPose=@n19, Alpha=@n14, Translation="3,0,0", Rotation="0,0,0", Scale="1,1,1") {
    rotationSpace=BCS_BoneSpace
    translationMode=BMM_Additive
    translationSpace=BCS_BoneSpace
    alphaBoolBlend {blendOption=Linear}
    alphaScaleBiasClamp {bInterpResult=true, interpSpeedDecreasing=7, interpSpeedIncreasing=4}
    boneToModify {boneName=slimeyB_03}
    translation {x=3}
   }
   n21 = "Transform (Modify) Bone - Bone: slimeyD_03"(ComponentPose=@n20, Alpha=@n14, Translation="3,0,0", Rotation="0,0,0", Scale="1,1,1") {
    rotationSpace=BCS_BoneSpace
    translationMode=BMM_Additive
    translationSpace=BCS_BoneSpace
    alphaBoolBlend {blendOption=Linear}
    alphaScaleBiasClamp {bInterpResult=true, interpSpeedDecreasing=7, interpSpeedIncreasing=4}
    boneToModify {boneName=slimeyD_03}
    translation {x=3}
   }
   n22 = "Transform (Modify) Bone - Bone: slimeyE_03"(ComponentPose=@n21, Alpha=@n14, Translation="5,0,0", Rotation="0,0,0", Scale="1,1,1") {
    rotationSpace=BCS_BoneSpace
    translationMode=BMM_Additive
    translationSpace=BCS_BoneSpace
    alphaBoolBlend {blendOption=Linear}
    alphaScaleBiasClamp {bInterpResult=true, interpSpeedDecreasing=7, interpSpeedIncreasing=4}
    boneToModify {boneName=slimeyE_03}
    translation {x=5}
   }
   Output Pose(Result=Blend Poses by bool(
     BlendPose_0=Sequence Player 'JoyJoy'() {sequence=/BRCosmetics/Accessories/FORT_Backpacks/Backpack_M_MED_Teriyaki_Warrior/Anims/JoyJoy}
     BlendPose_1=Component To Local(ComponentPose=@n22)
     BlendTime_0=0.2
     BlendTime_1=0.65
     bActiveValue=Get bShouldPlayJoyJoy() -> (bShouldPlayJoyJoy:bool)
    ) {blendType=Linear}) {layerGroup=None}
  ]}
 EventGraph {nodes [
   Event OnCharacterCustomizationLoaded()
   > n1 = Cast To FortPlayerPawn(Object=TryGetFortPawn())
   > Set pawn(pawn:Fort Player Pawn Object Reference=@n1) -> (Output_Get:Fort Player Pawn Object Reference)
   > n2 = WatchCosmeticStat!(
    execute:Exec
    Pawn:Fort Player Pawn Object Reference=Get pawn() -> (pawn:Fort Player Pawn Object Reference)
    StatTag:Fort Stat Manager Tag Structure=(TagName="StatManager.AthenaKills")
    DefaultValue:int=0
    bBroadcastInitialValue:bool=false
    InitialValueBroadcastDelay:float=-1
    ChangeValueBroadcastDelay:float=-1
    bConsumeWaitingNotifications:bool=true
   ) -> (then:Exec, OnStatChanged:Exec, StatName:Name, StatValue:int) {is=title K2Node_AsyncAction}
   n3 = "To Float (Integer)"(InInt=@n2.StatValue)
   n4 = "float > float"(A=@n3, B=Get KillCount() -> (KillCount:double)) {is=title K2Node_CallFunction}
   Branch(execute=@n2.OnStatChanged, Condition=@n4)
   > Set Active_Kill(Active_Kill:bool=@n4) -> (Output_Get:bool)
   > Set KillCount(KillCount:double=@n3) -> (Output_Get:double)
   > Branch(Condition=Get Active_Kill() -> (Active_Kill:bool))
   > Set bShouldPlayJoyJoy(bShouldPlayJoyJoy:bool=true) -> (Output_Get:bool)
   > RetriggerableDelay(Duration=3.4)
   > Set Active_Kill(Active_Kill=false)
   > Set bShouldPlayJoyJoy(bShouldPlayJoyJoy=false)
   > ResetDynamics(InTeleportType=ResetPhysics)
  ]}
]
variables [
 Active_Kill {type=boolean}
 bShouldPlayJoyJoy {type=boolean}
 KillCount {type=number}
 pawn {type=FortPlayerPawn}
]
